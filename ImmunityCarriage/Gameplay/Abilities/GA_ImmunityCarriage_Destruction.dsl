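; GA_ImmunityCarriage_Destruction is a gameplay ability for the Immunity Carriage that applies hit effects to continuously selected targets.
; On activation, a continuous target selection task named "Targeting" uses the Gameplay.Application.Hit tag and applies the gameplay effect container to each targeted result.
; Ending the ability stops that task by its instance name.
; The custom ability source transform comes from the avatar actor's transform.
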(event Ability|EventActivateAbility
  (bind (_asynctaskproxy _targetdata _applicationtag) (Ability|Tasks|DoTargetSelectionContinuous "Targeting" "(TagName=\"Gameplay.Application.Hit\")" false)
    (:Targeted
      (Ability|ApplyGameplayEffectContainer _targetdata _applicationtag))))

(event Ability|EventOnEndAbility
  (params (bWasCancelled :type Boolean))
  (Ability|EndTaskbyInstanceName "Targeting"))

(fn GetCustomAbilitySourceTransform
  (return (Transformation|GetActorTransform (Ability|GetAvatarActorfromActorInfo))))
